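# Info panel widget for the "Playing Now" tile in the Discovery Browser, built on CommonUserWidget.
# A vertical stack holds a secondary label, a row with the activity name and a Squeegee plugin slot for the rating tag, and a wrapping description.
# PreConstruct sets the label to a localized "Playing Now" in a green override color.
# Name and description hold placeholder text, and this graph does not fill them.
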WidgetBlueprint /DiscoveryBrowser/Tiles/PlayingNow/WBP_PlayingNow_Info
graphs [EventGraph {nodes [
   Event PreConstruct()
   n1 = Get TextSecondary_PlayingNow() -> (TextSecondary_PlayingNow:WBP Discover Tile Text Secondary Object Reference)
   > SetBlockText!(execute, self:WBP Discover Tile Text Secondary Object Reference=@n1, SecondaryText:Text="LOCTEXT(\"Playing Now\")") -> (then)
   > SetBlockTextColorAndOpacity!(
    execute
    self:WBP Discover Tile Text Secondary Object Reference=@n1
    UseDefault:bool=false
    OverrideColor_SpecifiedColor:Linear Color Structure=(R=0.0395,G=1,B=0.2307,A=1)
    OverrideColor_ColorUseRule:ESlateColorStylingMode Enum
   ) -> (then)
   Event Construct()
   Event Tick()
  ]}]
variables []
widgets {
 info {inheritedWidgetCount=0, parentClass=&CommonUserWidget, rootWidgetClass=&Overlay, widgetCount=7}
 widgets [
  Overlay_0 {widgetClassPath=&Overlay, widget=Overlay_0}
  StackBox_1 {
   widgetClassPath=&StackBox
   parent=Overlay_0
   widget=StackBox_1
   properties {orientation=Orient_Vertical}
   slot {class=OverlaySlot, properties {horizontalAlignment=HAlign_Fill, verticalAlignment=VAlign_Fill, padding=<ltrb 26 0 0 0>}}
  }
  TextSecondary_PlayingNow {
   widgetClassPath=/DiscoveryBrowser/Tiles/WBP_Discover_Tile_TextSecondary
   bIsVariable=true
   parent=StackBox_1
   widget=TextSecondary_PlayingNow
   slot {class=StackBoxSlot}
  }
  HorizontalBox_1 {
   widgetClassPath=&HorizontalBox
   parent=StackBox_1
   widget=HorizontalBox_1
   slot {class=StackBoxSlot, properties {padding=<ltrb 0 12 0 12>}}
  }
  Text_ActivityName {
   widgetClassPath=&CommonTextBlock
   bIsVariable=true
   parent=HorizontalBox_1
   widget=Text_ActivityName
   properties {
    applyLineHeightToBottomLine=false
    lineHeightPercentage=1.28
    mobileFontSizeMultiplier=1.2
    scrollStyle=&TextScrollStyle-Base_C
    text="{Activity Name} "
    wrapTextAt=800
    wrappingPolicy=AllowPerCharacterWrapping
    font {
     bForceMonospaced=false
     bMaterialIsStencil=false
     fontMaterial=null
     fontObject=&HeadingNow
     letterSpacing=-40
     monospacedWidth=1
     size=22.5
     skewAmount=0
     typefaceFontName=75 Medium
     outlineSettings {
      bApplyOutlineToDropShadows=false
      bMiteredCorners=false
      bSeparateFillAlpha=false
      outlineMaterial=null
      outlineSize=0
      outlineColor=<rgba 0 0 0 1>
     }
    }
    shadowOffset=<0 0>
    shapedTextOptions {bOverride_TextFlowDirection=true, bOverride_TextShapingMethod=false, textFlowDirection=Culture, textShapingMethod=Auto}
    strikeBrush {
     bIsDynamicallyLoaded=false
     drawAs=NoDrawType
     imageType=NoImage
     mirroring=NoMirror
     resourceName=None
     resourceObject=null
     tiling=NoTile
     imageSize=<32 32>
     margin=<ltrb 0 0 0 0>
     outlineSettings {
      bUseBrushTransparency=false
      roundingType=HalfHeightRadius
      width=0
      color {colorUseRule=UseColor_Specified, specifiedColor=<rgba 0 0 0 0>}
      cornerRadii=<0 0 0 0>
     }
     tintColor {colorUseRule=UseColor_Specified, specifiedColor=<rgba 1 1 1 1>}
     uVRegion {bIsValid=false, max=<0 0>, min=<0 0>}
    }
   }
   slot {class=HorizontalBoxSlot, properties {size {sizeRule=Fill, value=1}}}
  }
  SqueegeeSlot_Tag {
   widgetClassPath=/Game/UI/Plugin/Squeegee/SqueegeeInjectionSlot
   bIsVariable=true
   parent=HorizontalBox_1
   widget=SqueegeeSlot_Tag
   properties {
    pluginId=Squeegee
    pluginWidgetId=RatingTag
    styleParams {bShowDetails=true, fontSize=15, paddingSize=0.5, iconSize=<26 26>, iconSkewSize=<5 26>}
   }
   slot {class=HorizontalBoxSlot}
  }
  Text_ActivityDescription {
   widgetClassPath=&CommonTextBlock
   bIsVariable=true
   parent=StackBox_1
   widget=Text_ActivityDescription
   properties {
    applyLineHeightToBottomLine=false
    lineHeightPercentage=1.28
    mobileFontSizeMultiplier=1.2
    scrollStyle=&TextScrollStyle-Base_C
    text="[Game Mode Description.  Wrapping is set to per character, text widget clips to bounds on itself. Lorem ipsum dolor sit amet, consectetur adipiscing elit."
    wrapTextAt=700
    wrappingPolicy=AllowPerCharacterWrapping
    colorAndOpacity {colorUseRule=UseColor_Specified, specifiedColor=<rgba 0.5776 0.5776 0.5776 1>}
    font {
     bForceMonospaced=false
     bMaterialIsStencil=false
     fontMaterial=null
     fontObject=&HeadingNow
     letterSpacing=0
     monospacedWidth=1
     size=15
     skewAmount=0
     typefaceFontName=74 Regular
     outlineSettings {
      bApplyOutlineToDropShadows=false
      bMiteredCorners=false
      bSeparateFillAlpha=false
      outlineMaterial=null
      outlineSize=0
      outlineColor=<rgba 0 0 0 1>
     }
    }
    shadowOffset=<0 0>
    shapedTextOptions {bOverride_TextFlowDirection=true, bOverride_TextShapingMethod=false, textFlowDirection=Culture, textShapingMethod=Auto}
    strikeBrush {
     bIsDynamicallyLoaded=false
     drawAs=NoDrawType
     imageType=NoImage
     mirroring=NoMirror
     resourceName=None
     resourceObject=null
     tiling=NoTile
     imageSize=<32 32>
     margin=<ltrb 0 0 0 0>
     outlineSettings {
      bUseBrushTransparency=false
      roundingType=HalfHeightRadius
      width=0
      color {colorUseRule=UseColor_Specified, specifiedColor=<rgba 0 0 0 0>}
      cornerRadii=<0 0 0 0>
     }
     tintColor {colorUseRule=UseColor_Specified, specifiedColor=<rgba 1 1 1 1>}
     uVRegion {bIsValid=false, max=<0 0>, min=<0 0>}
    }
   }
   slot {class=StackBoxSlot, properties {size {sizeRule=Fill, value=1}}}
  }
 ]
}
