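# Playing Now tile for the Discovery browser, extending the minimal island tile with an info panel and a swappable background.
# PreConstruct collapses the base tile's outside-content attachment and spawns the background class into the nine-grid's back layer.
# Parent overrides feed name, description (trimmed after its first line break) and view-model tag into the info panel.
# A size-override delegate widens the tile container by the description's wrap width.
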WidgetBlueprint /DiscoveryBrowser/Tiles/PlayingNow/WBP_PlayingNow_Tile
graphs [
 Add Background {nodes [
   Add Background() {is=title K2Node_FunctionEntry}
   > RemoveWidget!(execute, Widget:Widget Object Reference=Get Block_Background() -> (Block_Background:WBP Playing Now Background Object Reference)) -> (then)
   > n1 = Construct Object from Class(Class=Get BackgroundClass() -> (BackgroundClass:WBP Playing Now Background Class Reference)) -> (ReturnValue:WBP Playing Now Background Object Reference)
   > n2 = Set Block_Background(Block_Background:WBP Playing Now Background Object Reference=@n1) -> (Output_Get:WBP Playing Now Background Object Reference)
   > Branch(Condition=IsValid(Object=@n2))
   n3 = Get Block_Background()
   > AddChildToGrid!(
    execute
    self:Grid Panel Object Reference=Get GridRoot_9Grid() -> (GridRoot_9Grid:Grid Panel Object Reference)
    Content:Widget Object Reference=@n3
    InRow:int=1
    InColumn:int=1
   ) -> (then, ReturnValue:Grid Slot Object Reference)
   > SetGridSlotLayer!(execute, Widget:Widget Object Reference=@n3, Layer:int=0) -> (then)
   > SetHorizontalAlignment(self:Grid Slot Object Reference=SlotAsGridSlot(Widget=@n3), InHorizontalAlignment=HAlign_Fill)
  ]}
 EventGraph {nodes [
   OnTileSizeSetOverrideContainerSize() {is=title K2Node_CustomEvent}
   n1 = Get WBP_Discover_Tile_Base() -> (WBP_Discover_Tile_Base:WBP Discover Tile Base Object Reference)
   n2 = BreakVector2D(InVec=Get Dynamic Tile Size!(self:WBP Discover Tile Base Object Reference=@n1) -> (DynamicTileSize:Vector 2D Structure) {is=title K2Node_CallFunction})
   n3 = Get WrapTextAt!(self:Text Layout Widget Object Reference=Get Text_ActivityDescription!(self:WBP Playing Now Info Object Reference=Get WBP_PlayingNow_Info_60() -> (WBP_PlayingNow_Info_60:WBP Playing Now Info Object Reference)) -> (Text_ActivityDescription:Common Text Object Reference)) -> (WrapTextAt:float)
   > Set Tile Container Size Override!(
    execute
    self:WBP Discover Tile Base Object Reference=@n1
    TileSizeOverrideWithOutsideContent:Vector 2D Structure=MakeVector2D(X=float + float(A=@n2, B=@n3), Y=@n2.Y)
   ) -> (then) {is=title K2Node_CallFunction}
   n4 = Event PreConstruct()
   > Parent: PreConstruct(IsDesignTime=@n4.IsDesignTime)
   > n5 = Sequence()
   > Branch(Condition=@n4.IsDesignTime)
   n6 = Get WBP_Discover_Tile_Base()
   > DesignTimeSetTestColumnSize!(execute, self:WBP Discover Tile Base Object Reference=@n6, NewSize:int=9, isSquareTiles:bool=false) -> (then)
   SetVisibility(
    execute=@n5.then_1
    self=Get VisualAttachment_OutsideContent!(self:WBP Discover Tile Base Object Reference=@n6) -> (VisualAttachment_OutsideContent:Fort Visual Attachment Object Reference)
    InVisibility=Collapsed
   )
   > Add Background() {is=title K2Node_CallFunction}
   Event Construct()
   > SetCornerRadius!(
    execute
    self:WBP Island Tile Minimal Object Reference
    CornerRadius_R:float=20
    CornerRadius_G:float=20
    CornerRadius_B:float=20
    CornerRadius_A:float=20
   ) -> (then)
   > Parent: Construct()
   n7 = Get WBP_Discover_Tile_Base()
   > Bind Event to On Set Tile Container Size Override(self:WBP Discover Tile Base Object Reference=@n7, Delegate=Create Event()) {is=title K2Node_AddDelegate}
   Event Destruct()
   > Parent: Destruct()
   > Unbind all Events from On Set Tile Container Size Override!(execute:Exec, self:WBP Discover Tile Base Object Reference=@n7) -> (then:Exec) {is=title K2Node_ClearDelegate}
  ]}
 RemoveIndentation {nodes [
   n1 = RemoveIndentation!() -> (then, InputText:Text) {is=title K2Node_FunctionEntry}
   n2 = "To String (Text)"(InText=@n1)
   n3 = Split(SourceString=@n2, InStr="\r\n", SearchDir=FromStart)
   n4 = Append!(A=@n3, B="\r\n", C:String=Trim!(SourceString:String=@n3.RightS) -> (ReturnValue:String)) -> (ReturnValue)
   n5 = "integer > integer"(A=FindSubstring(SearchIn=@n2, Substring="\r\n"))
   > Return Node!(execute, OutputText:Text="To Text (String)"(InString=SelectString(A=@n4, B=@n2, bPickA=@n5)))
  ]}
 SetIslandDescription {nodes [
   n1 = SetIslandDescription!() -> (then, IslandDescription:Text) {is=title K2Node_FunctionEntry}
   > Parent: SetIslandDescription!(execute, IslandDescription:Text=@n1) -> (then) {is=title K2Node_CallParentFunction}
   > n2 = RemoveIndentation!(execute, self, InputText:Text=@n1) -> (then, OutputText:Text) {is=title K2Node_CallFunction}
   > "SetText (Text)"(self=Get Text_ActivityDescription(self=Get WBP_PlayingNow_Info_60()), InText=@n2)
  ]}
 SetIslandName {nodes [
   n1 = SetIslandName!() -> (then, IslandName:Text) {is=title K2Node_FunctionEntry}
   > Parent: SetIslandName!(execute, IslandName:Text=@n1) -> (then) {is=title K2Node_CallParentFunction}
   > "SetText (Text)"(
    self=Get Text_ActivityName!(self:WBP Playing Now Info Object Reference=Get WBP_PlayingNow_Info_60()) -> (Text_ActivityName:Common Text Object Reference)
    InText=@n1
   )
  ]}
 SetupSqueegeeWidgets {nodes [
   n1 = SetupSqueegeeWidgets!() -> (then, ActivityVM:Discover Activity View Model Object Reference) {is=title K2Node_FunctionEntry}
   > Parent: SetupSqueegeeWidgets!(execute, ActivityVM:Discover Activity View Model Object Reference=@n1) -> (then) {is=title K2Node_CallParentFunction}
   > ShowActivityDataFromViewModel!(
    execute
    self:Fort Squeegee Widget Injection Slot Object Reference=Get SqueegeeSlot_Tag!(self:WBP Playing Now Info Object Reference=Get WBP_PlayingNow_Info_60()) -> (SqueegeeSlot_Tag:Squeegee Injection Slot Object Reference)
    GameActivityViewModel:Fort Discover Link Entry Tile Item VM Object Reference=@n1
   ) -> (then)
  ]}
 TileSizeSet {nodes [TileSizeSet() {is=title K2Node_FunctionEntry}]}
]
variables [
 BackgroundClass {type=Class, default=&/WBP_PlayingNow_Background.WBP_PlayingNow_Background_C}
 Block_Background {type=WBP_PlayingNow_Background_C, default=null}
 TileSizeSet {}
]
widgets {
 info {
  inheritedWidgetCount=0
  namedSlotCount=1
  parentClass=/DiscoveryBrowser/Tiles/Islands/WBP_IslandTile_Minimal.WBP_IslandTile_Minimal_C
  rootWidgetClass=null
  widgetCount=1
 }
 widgets [WBP_PlayingNow_Info_60 {widgetClassPath=&/WBP_PlayingNow_Info, bIsVariable=true, widget=WBP_PlayingNow_Info_60}]
}
